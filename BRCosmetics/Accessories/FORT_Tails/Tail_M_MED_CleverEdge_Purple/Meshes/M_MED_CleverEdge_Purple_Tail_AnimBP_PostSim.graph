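# Post-process Anim Blueprint for the CleverEdge Purple tail cosmetic, adding trailing motion to its bone chain.
# The output runs the cached input pose through a Trail controller driven by TrailAlpha and RelaxSpeed.
# Additive bone-pose chains blended by crouch, fall, jump and movement flags exist but never reach Output Pose.
# On customization load, a backbling or cape sets the material's "Hide Element 1" parameter.
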AnimBlueprint /BRCosmetics/Accessories/FORT_Tails/Tail_M_MED_CleverEdge_Purple/Meshes/M_MED_CleverEdge_Purple_Tail_AnimBP_PostSim
target_skeleton=/Game/Accessories/FORT_Tails/Common/Fortnite_Base_Tail_Skeleton
graphs [
 AnimGraph {nodes [
   n1 = "Transform (Modify) Bone - Bone: C_Tail_A_Base_Jnt"(Alpha=1, Translation="0,0,0", Rotation="0,30,0", Scale="1,1,1") {
    rotationMode=BMM_Additive
    rotationSpace=BCS_BoneSpace
    translationMode=BMM_Additive
    translationSpace=BCS_BoneSpace
    alphaBoolBlend {blendOption=Linear}
    boneToModify {boneName=C_Tail_A_Base_Jnt}
    rotation {yaw=30}
   }
   n2 = "Transform (Modify) Bone - Bone: C_Tail_A_1_Jnt"(ComponentPose=@n1, Alpha=1, Translation="0,0,0", Rotation="5,0,0", Scale="1,1,1") {
    rotationMode=BMM_Additive
    rotationSpace=BCS_BoneSpace
    translationMode=BMM_Additive
    translationSpace=BCS_BoneSpace
    alphaBoolBlend {blendOption=Linear}
    boneToModify {boneName=C_Tail_A_1_Jnt}
    rotation {pitch=5}
   }
   n3 = "Transform (Modify) Bone - Bone: C_Tail_A_2_Jnt"(ComponentPose=@n2, Alpha=1, Translation="0,0,0", Rotation="5,0,0", Scale="1,1,1") {
    rotationMode=BMM_Additive
    rotationSpace=BCS_BoneSpace
    translationMode=BMM_Additive
    translationSpace=BCS_BoneSpace
    alphaBoolBlend {blendOption=Linear}
    boneToModify {boneName=C_Tail_A_2_Jnt}
    rotation {pitch=5}
   }
   n4 = "Transform (Modify) Bone - Bone: C_Tail_A_3_Jnt"(ComponentPose=@n3, Alpha=1, Translation="0,0,0", Rotation="5,0,0", Scale="1,1,1") {
    rotationMode=BMM_Additive
    rotationSpace=BCS_BoneSpace
    translationMode=BMM_Additive
    translationSpace=BCS_BoneSpace
    alphaBoolBlend {blendOption=Linear}
    boneToModify {boneName=C_Tail_A_3_Jnt}
    rotation {pitch=5}
   }
   n5 = "Transform (Modify) Bone - Bone: C_Tail_A_4_Jnt"(ComponentPose=@n4, Alpha=1, Translation="0,0,0", Rotation="10,0,0", Scale="1,1,1") {
    rotationMode=BMM_Additive
    rotationSpace=BCS_BoneSpace
    translationMode=BMM_Additive
    translationSpace=BCS_BoneSpace
    alphaBoolBlend {blendOption=Linear}
    boneToModify {boneName=C_Tail_A_4_Jnt}
    rotation {pitch=10}
   }
   n6 = "Transform (Modify) Bone - Bone: C_Tail_A_5_Jnt"(ComponentPose=@n5, Alpha=1, Translation="0,0,0", Rotation="10,0,0", Scale="1,1,1") {
    rotationMode=BMM_Additive
    rotationSpace=BCS_BoneSpace
    translationMode=BMM_Additive
    translationSpace=BCS_BoneSpace
    alphaBoolBlend {blendOption=Linear}
    boneToModify {boneName=C_Tail_A_5_Jnt}
    rotation {pitch=10}
   }
   n7 = "Transform (Modify) Bone - Bone: C_Tail_A_6_Jnt"(ComponentPose=@n6, Alpha=1, Translation="0,0,0", Rotation="15,0,0", Scale="1,1,1") {
    rotationMode=BMM_Additive
    rotationSpace=BCS_BoneSpace
    translationMode=BMM_Additive
    translationSpace=BCS_BoneSpace
    alphaBoolBlend {blendOption=Linear}
    boneToModify {boneName=C_Tail_A_6_Jnt}
    rotation {pitch=15}
   }
   n8 = "Transform (Modify) Bone - Bone: C_Tail_A_7_Jnt"(ComponentPose=@n7, Alpha=1, Translation="0,0,0", Rotation="20,0,0", Scale="1,1,1") {
    rotationMode=BMM_Additive
    rotationSpace=BCS_BoneSpace
    translationMode=BMM_Additive
    translationSpace=BCS_BoneSpace
    alphaBoolBlend {blendOption=Linear}
    boneToModify {boneName=C_Tail_A_7_Jnt}
    rotation {pitch=20}
   }
   n9 = "Transform (Modify) Bone - Bone: C_Tail_A_8_Jnt"(ComponentPose=@n8, Alpha=1, Translation="0,0,0", Rotation="20,0,0", Scale="1,1,1") {
    rotationMode=BMM_Additive
    rotationSpace=BCS_BoneSpace
    translationMode=BMM_Additive
    translationSpace=BCS_BoneSpace
    alphaBoolBlend {blendOption=Linear}
    boneToModify {boneName=C_Tail_A_8_Jnt}
    rotation {pitch=20}
   }
   n10 = "Transform (Modify) Bone - Bone: C_Tail_A_9_Jnt"(ComponentPose=@n9, Alpha=1, Translation="0,0,0", Rotation="40,0,0", Scale="1,1,1") {
    rotationMode=BMM_Additive
    rotationSpace=BCS_BoneSpace
    translationMode=BMM_Additive
    translationSpace=BCS_BoneSpace
    alphaBoolBlend {blendOption=Linear}
    boneToModify {boneName=C_Tail_A_9_Jnt}
    rotation {pitch=40}
   }
   n11 = Component To Local(ComponentPose=@n10)
   n12 = "Transform (Modify) Bone - Bone: C_Tail_A_Base_Jnt"(ComponentPose=Local To Component(LocalPose=Use cached pose 'CachePose'()), Alpha=1, Translation="0,0,0", Rotation="0,40,0", Scale="1,1,1") {
    rotationMode=BMM_Additive
    rotationSpace=BCS_BoneSpace
    translationMode=BMM_Additive
    translationSpace=BCS_BoneSpace
    alphaBoolBlend {blendOption=Linear}
    boneToModify {boneName=C_Tail_A_Base_Jnt}
    rotation {yaw=40}
   }
   n13 = "Transform (Modify) Bone - Bone: C_Tail_A_1_Jnt"(ComponentPose=@n12, Alpha=1, Translation="0,0,0", Rotation="30,0,0", Scale="1,1,1") {
    rotationMode=BMM_Additive
    rotationSpace=BCS_BoneSpace
    translationMode=BMM_Additive
    translationSpace=BCS_BoneSpace
    alphaBoolBlend {blendOption=Linear}
    boneToModify {boneName=C_Tail_A_1_Jnt}
    rotation {pitch=30}
   }
   n14 = "Transform (Modify) Bone - Bone: C_Tail_A_2_Jnt"(ComponentPose=@n13, Alpha=1, Translation="0,0,0", Rotation="20,0,0", Scale="1,1,1") {
    rotationMode=BMM_Additive
    rotationSpace=BCS_BoneSpace
    translationMode=BMM_Additive
    translationSpace=BCS_BoneSpace
    alphaBoolBlend {blendOption=Linear}
    boneToModify {boneName=C_Tail_A_2_Jnt}
    rotation {pitch=20}
   }
   n15 = "Transform (Modify) Bone - Bone: C_Tail_A_3_Jnt"(ComponentPose=@n14, Alpha=1, Translation="0,0,0", Rotation="10,0,0", Scale="1,1,1") {
    rotationMode=BMM_Additive
    rotationSpace=BCS_BoneSpace
    translationMode=BMM_Additive
    translationSpace=BCS_BoneSpace
    alphaBoolBlend {blendOption=Linear}
    boneToModify {boneName=C_Tail_A_3_Jnt}
    rotation {pitch=10}
   }
   n16 = "Transform (Modify) Bone - Bone: C_Tail_A_4_Jnt"(ComponentPose=@n15, Alpha=1, Translation="0,0,0", Rotation="10,0,0", Scale="1,1,1") {
    rotationMode=BMM_Additive
    rotationSpace=BCS_BoneSpace
    translationMode=BMM_Additive
    translationSpace=BCS_BoneSpace
    alphaBoolBlend {blendOption=Linear}
    boneToModify {boneName=C_Tail_A_4_Jnt}
    rotation {pitch=10}
   }
   n17 = "Transform (Modify) Bone - Bone: C_Tail_A_5_Jnt"(ComponentPose=@n16, Alpha=1, Translation="0,0,0", Rotation="0,0,0", Scale="1,1,1") {
    rotationMode=BMM_Additive
    rotationSpace=BCS_BoneSpace
    translationMode=BMM_Additive
    translationSpace=BCS_BoneSpace
    alphaBoolBlend {blendOption=Linear}
    boneToModify {boneName=C_Tail_A_5_Jnt}
   }
   n18 = "Transform (Modify) Bone - Bone: C_Tail_A_6_Jnt"(ComponentPose=@n17, Alpha=1, Translation="0,0,0", Rotation="0,0,0", Scale="1,1,1") {
    rotationMode=BMM_Additive
    rotationSpace=BCS_BoneSpace
    translationMode=BMM_Additive
    translationSpace=BCS_BoneSpace
    alphaBoolBlend {blendOption=Linear}
    boneToModify {boneName=C_Tail_A_6_Jnt}
   }
   n19 = "Transform (Modify) Bone - Bone: C_Tail_A_7_Jnt"(ComponentPose=@n18, Alpha=1, Translation="0,0,0", Rotation="0,0,0", Scale="1,1,1") {
    rotationMode=BMM_Additive
    rotationSpace=BCS_BoneSpace
    translationMode=BMM_Additive
    translationSpace=BCS_BoneSpace
    alphaBoolBlend {blendOption=Linear}
    boneToModify {boneName=C_Tail_A_7_Jnt}
   }
   n20 = "Transform (Modify) Bone - Bone: C_Tail_A_8_Jnt"(ComponentPose=@n19, Alpha=1, Translation="0,0,0", Rotation="0,0,0", Scale="1,1,1") {
    rotationMode=BMM_Additive
    rotationSpace=BCS_BoneSpace
    translationMode=BMM_Additive
    translationSpace=BCS_BoneSpace
    alphaBoolBlend {blendOption=Linear}
    boneToModify {boneName=C_Tail_A_8_Jnt}
   }
   n21 = "Transform (Modify) Bone - Bone: C_Tail_A_9_Jnt"(ComponentPose=@n20, Alpha=1, Translation="0,0,0", Rotation="0,0,0", Scale="1,1,1") {
    rotationMode=BMM_Additive
    rotationSpace=BCS_BoneSpace
    translationMode=BMM_Additive
    translationSpace=BCS_BoneSpace
    alphaBoolBlend {blendOption=Linear}
    alphaScaleBiasClamp {bInterpResult=true}
    boneToModify {boneName=C_Tail_A_9_Jnt}
   }
   n22 = Component To Local(ComponentPose=@n21)
   n23 = Get Is Moving 2D() -> (Is Moving 2D:bool)
   n24 = "Transform (Modify) Bone - Bone: C_Tail_A_Base_Jnt"(ComponentPose=Local To Component(LocalPose=Use cached pose 'CachePose'()), Alpha=1, Translation="0,0,0", Rotation="0,30,0", Scale="1,1,1") {
    rotationMode=BMM_Additive
    rotationSpace=BCS_BoneSpace
    translationMode=BMM_Additive
    translationSpace=BCS_BoneSpace
    alphaBoolBlend {blendOption=Linear}
    alphaScaleBiasClamp {bInterpResult=true, interpSpeedDecreasing=1, interpSpeedIncreasing=1, scale=0.3}
    boneToModify {boneName=C_Tail_A_Base_Jnt}
    rotation {yaw=30}
   }
   n25 = "Transform (Modify) Bone - Bone: C_Tail_A_1_Jnt"(ComponentPose=@n24, Alpha=1, Translation="0,0,0", Rotation="-10,0,0", Scale="1,1,1") {
    rotationMode=BMM_Additive
    rotationSpace=BCS_BoneSpace
    translationMode=BMM_Additive
    translationSpace=BCS_BoneSpace
    alphaBoolBlend {blendOption=Linear}
    alphaScaleBiasClamp {bInterpResult=true, interpSpeedDecreasing=1, interpSpeedIncreasing=1, scale=0.3}
    boneToModify {boneName=C_Tail_A_1_Jnt}
    rotation {pitch=-10}
   }
   n26 = "Transform (Modify) Bone - Bone: C_Tail_A_2_Jnt"(ComponentPose=@n25, Alpha=1, Translation="0,0,0", Rotation="-10,0,0", Scale="1,1,1") {
    rotationMode=BMM_Additive
    rotationSpace=BCS_BoneSpace
    translationMode=BMM_Additive
    translationSpace=BCS_BoneSpace
    alphaBoolBlend {blendOption=Linear}
    alphaScaleBiasClamp {bInterpResult=true, interpSpeedDecreasing=1, interpSpeedIncreasing=1, scale=0.3}
    boneToModify {boneName=C_Tail_A_2_Jnt}
    rotation {pitch=-10}
   }
   n27 = "Transform (Modify) Bone - Bone: C_Tail_A_3_Jnt"(ComponentPose=@n26, Alpha=1, Translation="0,0,0", Rotation="-10,0,0", Scale="1,1,1") {
    rotationMode=BMM_Additive
    rotationSpace=BCS_BoneSpace
    translationMode=BMM_Additive
    translationSpace=BCS_BoneSpace
    alphaBoolBlend {blendOption=Linear}
    alphaScaleBiasClamp {bInterpResult=true, interpSpeedDecreasing=1, interpSpeedIncreasing=1, scale=0.3}
    boneToModify {boneName=C_Tail_A_3_Jnt}
    rotation {pitch=-10}
   }
   n28 = "Transform (Modify) Bone - Bone: C_Tail_A_4_Jnt"(ComponentPose=@n27, Alpha=1, Translation="0,0,0", Rotation="-10,0,0", Scale="1,1,1") {
    rotationMode=BMM_Additive
    rotationSpace=BCS_BoneSpace
    translationMode=BMM_Additive
    translationSpace=BCS_BoneSpace
    alphaBoolBlend {blendOption=Linear}
    alphaScaleBiasClamp {bInterpResult=true, interpSpeedDecreasing=1, interpSpeedIncreasing=1, scale=0.3}
    boneToModify {boneName=C_Tail_A_4_Jnt}
    rotation {pitch=-10}
   }
   n29 = "Transform (Modify) Bone - Bone: C_Tail_A_5_Jnt"(ComponentPose=@n28, Alpha=1, Translation="0,0,0", Rotation="0,0,0", Scale="1,1,1") {
    rotationMode=BMM_Additive
    rotationSpace=BCS_BoneSpace
    translationMode=BMM_Additive
    translationSpace=BCS_BoneSpace
    alphaBoolBlend {blendOption=Linear}
    alphaScaleBiasClamp {bInterpResult=true, interpSpeedDecreasing=1, interpSpeedIncreasing=1, scale=0.3}
    boneToModify {boneName=C_Tail_A_5_Jnt}
   }
   n30 = "Transform (Modify) Bone - Bone: C_Tail_A_6_Jnt"(ComponentPose=@n29, Alpha=1, Translation="0,0,0", Rotation="0,0,0", Scale="1,1,1") {
    rotationMode=BMM_Additive
    rotationSpace=BCS_BoneSpace
    translationMode=BMM_Additive
    translationSpace=BCS_BoneSpace
    alphaBoolBlend {blendOption=Linear}
    alphaScaleBiasClamp {interpSpeedDecreasing=1, interpSpeedIncreasing=1}
    boneToModify {boneName=C_Tail_A_6_Jnt}
   }
   n31 = "Transform (Modify) Bone - Bone: C_Tail_A_7_Jnt"(ComponentPose=@n30, Alpha=1, Translation="0,0,0", Rotation="0,0,0", Scale="1,1,1") {
    rotationMode=BMM_Additive
    rotationSpace=BCS_BoneSpace
    translationMode=BMM_Additive
    translationSpace=BCS_BoneSpace
    alphaBoolBlend {blendOption=Linear}
    alphaScaleBiasClamp {interpSpeedDecreasing=1, interpSpeedIncreasing=1}
    boneToModify {boneName=C_Tail_A_7_Jnt}
   }
   n32 = "Transform (Modify) Bone - Bone: C_Tail_A_8_Jnt"(ComponentPose=@n31, Alpha=1, Translation="0,0,0", Rotation="0,0,0", Scale="1,1,1") {
    rotationMode=BMM_Additive
    rotationSpace=BCS_BoneSpace
    translationMode=BMM_Additive
    translationSpace=BCS_BoneSpace
    alphaBoolBlend {blendOption=Linear}
    alphaScaleBiasClamp {interpSpeedDecreasing=1, interpSpeedIncreasing=1}
    boneToModify {boneName=C_Tail_A_8_Jnt}
   }
   n33 = "Transform (Modify) Bone - Bone: C_Tail_A_9_Jnt"(ComponentPose=@n32, Alpha=1, Translation="0,0,0", Rotation="0,0,0", Scale="1,1,1") {
    rotationMode=BMM_Additive
    rotationSpace=BCS_BoneSpace
    translationMode=BMM_Additive
    translationSpace=BCS_BoneSpace
    alphaBoolBlend {blendOption=Linear}
    alphaScaleBiasClamp {bInterpResult=true, interpSpeedDecreasing=1, interpSpeedIncreasing=1, scale=0.3}
    boneToModify {boneName=C_Tail_A_9_Jnt}
   }
   n34 = "Transform (Modify) Bone - Bone: C_Tail_A_Base_Jnt"(ComponentPose=Local To Component(LocalPose=Use cached pose 'CachePose'()), Alpha=1, Translation="0,0,0", Rotation="0,30,0", Scale="1,1,1") {
    rotationMode=BMM_Additive
    rotationSpace=BCS_BoneSpace
    translationMode=BMM_Additive
    translationSpace=BCS_BoneSpace
    alphaBoolBlend {blendOption=Linear}
    boneToModify {boneName=C_Tail_A_Base_Jnt}
    rotation {yaw=30}
   }
   n35 = "Transform (Modify) Bone - Bone: C_Tail_A_1_Jnt"(ComponentPose=@n34, Alpha=1, Translation="0,0,0", Rotation="-20,0,0", Scale="1,1,1") {
    rotationMode=BMM_Additive
    rotationSpace=BCS_BoneSpace
    translationMode=BMM_Additive
    translationSpace=BCS_BoneSpace
    alphaBoolBlend {blendOption=Linear}
    boneToModify {boneName=C_Tail_A_1_Jnt}
    rotation {pitch=-20}
   }
   n36 = "Transform (Modify) Bone - Bone: C_Tail_A_2_Jnt"(ComponentPose=@n35, Alpha=1, Translation="0,0,0", Rotation="-20,0,0", Scale="1,1,1") {
    rotationMode=BMM_Additive
    rotationSpace=BCS_BoneSpace
    translationMode=BMM_Additive
    translationSpace=BCS_BoneSpace
    alphaBoolBlend {blendOption=Linear}
    boneToModify {boneName=C_Tail_A_2_Jnt}
    rotation {pitch=-20}
   }
   n37 = "Transform (Modify) Bone - Bone: C_Tail_A_3_Jnt"(ComponentPose=@n36, Alpha=1, Translation="0,0,0", Rotation="-20,0,0", Scale="1,1,1") {
    rotationMode=BMM_Additive
    rotationSpace=BCS_BoneSpace
    translationMode=BMM_Additive
    translationSpace=BCS_BoneSpace
    alphaBoolBlend {blendOption=Linear}
    boneToModify {boneName=C_Tail_A_3_Jnt}
    rotation {pitch=-20}
   }
   n38 = "Transform (Modify) Bone - Bone: C_Tail_A_4_Jnt"(ComponentPose=@n37, Alpha=1, Translation="0,0,0", Rotation="-20,0,0", Scale="1,1,1") {
    rotationMode=BMM_Additive
    rotationSpace=BCS_BoneSpace
    translationMode=BMM_Additive
    translationSpace=BCS_BoneSpace
    alphaBoolBlend {blendOption=Linear}
    boneToModify {boneName=C_Tail_A_4_Jnt}
    rotation {pitch=-20}
   }
   n39 = "Transform (Modify) Bone - Bone: C_Tail_A_5_Jnt"(ComponentPose=@n38, Alpha=1, Translation="0,0,0", Rotation="-15,0,0", Scale="1,1,1") {
    rotationMode=BMM_Additive
    rotationSpace=BCS_BoneSpace
    translationMode=BMM_Additive
    translationSpace=BCS_BoneSpace
    alphaBoolBlend {blendOption=Linear}
    boneToModify {boneName=C_Tail_A_5_Jnt}
    rotation {pitch=-15}
   }
   n40 = "Transform (Modify) Bone - Bone: C_Tail_A_6_Jnt"(ComponentPose=@n39, Alpha=1, Translation="0,0,0", Rotation="-10,0,0", Scale="1,1,1") {
    rotationMode=BMM_Additive
    rotationSpace=BCS_BoneSpace
    translationMode=BMM_Additive
    translationSpace=BCS_BoneSpace
    alphaBoolBlend {blendOption=Linear}
    boneToModify {boneName=C_Tail_A_6_Jnt}
    rotation {pitch=-10}
   }
   n41 = "Transform (Modify) Bone - Bone: C_Tail_A_7_Jnt"(ComponentPose=@n40, Alpha=1, Translation="0,0,0", Rotation="-10,0,0", Scale="1,1,1") {
    rotationMode=BMM_Additive
    rotationSpace=BCS_BoneSpace
    translationMode=BMM_Additive
    translationSpace=BCS_BoneSpace
    alphaBoolBlend {blendOption=Linear}
    boneToModify {boneName=C_Tail_A_7_Jnt}
    rotation {pitch=-10}
   }
   n42 = "Transform (Modify) Bone - Bone: C_Tail_A_8_Jnt"(ComponentPose=@n41, Alpha=1, Translation="0,0,0", Rotation="-10,0,0", Scale="1,1,1") {
    rotationMode=BMM_Additive
    rotationSpace=BCS_BoneSpace
    translationMode=BMM_Additive
    translationSpace=BCS_BoneSpace
    alphaBoolBlend {blendOption=Linear}
    boneToModify {boneName=C_Tail_A_8_Jnt}
    rotation {pitch=-10}
   }
   n43 = "Transform (Modify) Bone - Bone: C_Tail_A_9_Jnt"(ComponentPose=@n42, Alpha=1, Translation="0,0,0", Rotation="-10,0,0", Scale="1,1,1") {
    rotationMode=BMM_Additive
    rotationSpace=BCS_BoneSpace
    translationMode=BMM_Additive
    translationSpace=BCS_BoneSpace
    alphaBoolBlend {blendOption=Linear}
    alphaScaleBiasClamp {bInterpResult=true}
    boneToModify {boneName=C_Tail_A_9_Jnt}
    rotation {pitch=-10}
   }
   n44 = Blend Poses by bool(
    BlendPose_0=Component To Local(ComponentPose=@n43)
    BlendPose_1=Blend Poses by bool(
     BlendPose_0=Component To Local(ComponentPose=@n33)
     BlendPose_1=Blend Poses by bool(BlendPose_0=@n22, BlendPose_1=@n11, BlendTime_0=2, BlendTime_1=2, bActiveValue=@n23) {blendType=Linear}
     BlendTime_0=0.6
     BlendTime_1=0.6
     bActiveValue=Get bIsJumping()
    ) {blendType=Linear}
    BlendTime_0=0.6
    BlendTime_1=0.6
    bActiveValue=Get Is Falling() -> (Is Falling:bool)
   ) {blendType=Linear}
   n45 = "Transform (Modify) Bone - Bone: C_Tail_A_Base_Jnt"(ComponentPose=Local To Component(LocalPose=Use cached pose 'CachePose'()), Alpha=1, Translation="0,0,0", Rotation="0,30,0", Scale="1,1,1") {
    rotationMode=BMM_Additive
    rotationSpace=BCS_BoneSpace
    translationMode=BMM_Additive
    translationSpace=BCS_BoneSpace
    alphaBoolBlend {blendOption=Linear}
    alphaScaleBiasClamp {bInterpResult=true, interpSpeedDecreasing=1, interpSpeedIncreasing=1, scale=0.3}
    boneToModify {boneName=C_Tail_A_Base_Jnt}
    rotation {yaw=30}
   }
   n46 = "Transform (Modify) Bone - Bone: C_Tail_A_1_Jnt"(ComponentPose=@n45, Alpha=1, Translation="0,0,0", Rotation="-10,0,0", Scale="1,1,1") {
    rotationMode=BMM_Additive
    rotationSpace=BCS_BoneSpace
    translationMode=BMM_Additive
    translationSpace=BCS_BoneSpace
    alphaBoolBlend {blendOption=Linear}
    alphaScaleBiasClamp {bInterpResult=true, interpSpeedDecreasing=1, interpSpeedIncreasing=1, scale=0.3}
    boneToModify {boneName=C_Tail_A_1_Jnt}
    rotation {pitch=-10}
   }
   n47 = "Transform (Modify) Bone - Bone: C_Tail_A_2_Jnt"(ComponentPose=@n46, Alpha=1, Translation="0,0,0", Rotation="-10,0,0", Scale="1,1,1") {
    rotationMode=BMM_Additive
    rotationSpace=BCS_BoneSpace
    translationMode=BMM_Additive
    translationSpace=BCS_BoneSpace
    alphaBoolBlend {blendOption=Linear}
    alphaScaleBiasClamp {bInterpResult=true, interpSpeedDecreasing=1, interpSpeedIncreasing=1, scale=0.3}
    boneToModify {boneName=C_Tail_A_2_Jnt}
    rotation {pitch=-10}
   }
   n48 = "Transform (Modify) Bone - Bone: C_Tail_A_3_Jnt"(ComponentPose=@n47, Alpha=1, Translation="0,0,0", Rotation="-10,0,0", Scale="1,1,1") {
    rotationMode=BMM_Additive
    rotationSpace=BCS_BoneSpace
    translationMode=BMM_Additive
    translationSpace=BCS_BoneSpace
    alphaBoolBlend {blendOption=Linear}
    alphaScaleBiasClamp {bInterpResult=true, interpSpeedDecreasing=1, interpSpeedIncreasing=1, scale=0.3}
    boneToModify {boneName=C_Tail_A_3_Jnt}
    rotation {pitch=-10}
   }
   n49 = "Transform (Modify) Bone - Bone: C_Tail_A_4_Jnt"(ComponentPose=@n48, Alpha=1, Translation="0,0,0", Rotation="-10,0,0", Scale="1,1,1") {
    rotationMode=BMM_Additive
    rotationSpace=BCS_BoneSpace
    translationMode=BMM_Additive
    translationSpace=BCS_BoneSpace
    alphaBoolBlend {blendOption=Linear}
    alphaScaleBiasClamp {bInterpResult=true, interpSpeedDecreasing=1, interpSpeedIncreasing=1, scale=0.3}
    boneToModify {boneName=C_Tail_A_4_Jnt}
    rotation {pitch=-10}
   }
   n50 = "Transform (Modify) Bone - Bone: C_Tail_A_5_Jnt"(ComponentPose=@n49, Alpha=1, Translation="0,0,0", Rotation="0,0,0", Scale="1,1,1") {
    rotationMode=BMM_Additive
    rotationSpace=BCS_BoneSpace
    translationMode=BMM_Additive
    translationSpace=BCS_BoneSpace
    alphaBoolBlend {blendOption=Linear}
    alphaScaleBiasClamp {bInterpResult=true, interpSpeedDecreasing=1, interpSpeedIncreasing=1, scale=0.3}
    boneToModify {boneName=C_Tail_A_5_Jnt}
   }
   n51 = "Transform (Modify) Bone - Bone: C_Tail_A_6_Jnt"(ComponentPose=@n50, Alpha=1, Translation="0,0,0", Rotation="0,0,0", Scale="1,1,1") {
    rotationMode=BMM_Additive
    rotationSpace=BCS_BoneSpace
    translationMode=BMM_Additive
    translationSpace=BCS_BoneSpace
    alphaBoolBlend {blendOption=Linear}
    alphaScaleBiasClamp {interpSpeedDecreasing=1, interpSpeedIncreasing=1}
    boneToModify {boneName=C_Tail_A_6_Jnt}
   }
   n52 = "Transform (Modify) Bone - Bone: C_Tail_A_7_Jnt"(ComponentPose=@n51, Alpha=1, Translation="0,0,0", Rotation="0,0,0", Scale="1,1,1") {
    rotationMode=BMM_Additive
    rotationSpace=BCS_BoneSpace
    translationMode=BMM_Additive
    translationSpace=BCS_BoneSpace
    alphaBoolBlend {blendOption=Linear}
    alphaScaleBiasClamp {interpSpeedDecreasing=1, interpSpeedIncreasing=1}
    boneToModify {boneName=C_Tail_A_7_Jnt}
   }
   n53 = "Transform (Modify) Bone - Bone: C_Tail_A_8_Jnt"(ComponentPose=@n52, Alpha=1, Translation="0,0,0", Rotation="0,0,0", Scale="1,1,1") {
    rotationMode=BMM_Additive
    rotationSpace=BCS_BoneSpace
    translationMode=BMM_Additive
    translationSpace=BCS_BoneSpace
    alphaBoolBlend {blendOption=Linear}
    alphaScaleBiasClamp {interpSpeedDecreasing=1, interpSpeedIncreasing=1}
    boneToModify {boneName=C_Tail_A_8_Jnt}
   }
   n54 = "Transform (Modify) Bone - Bone: C_Tail_A_9_Jnt"(ComponentPose=@n53, Alpha=1, Translation="0,0,0", Rotation="0,0,0", Scale="1,1,1") {
    rotationMode=BMM_Additive
    rotationSpace=BCS_BoneSpace
    translationMode=BMM_Additive
    translationSpace=BCS_BoneSpace
    alphaBoolBlend {blendOption=Linear}
    alphaScaleBiasClamp {bInterpResult=true, interpSpeedDecreasing=1, interpSpeedIncreasing=1, scale=0.3}
    boneToModify {boneName=C_Tail_A_9_Jnt}
   }
   n55 = Component To Local(ComponentPose=@n54)
   n56 = AND Boolean(A=Get Is Jumping() -> (Is Jumping:bool), B=Get Is Moving 2D())
   n57 = "Transform (Modify) Bone - Bone: C_Tail_A_Base_Jnt"(ComponentPose=Local To Component(LocalPose=Use cached pose 'CachePose'()), Alpha=1, Translation="0,0,0", Rotation="0,35,0", Scale="1,1,1") {
    rotationMode=BMM_Additive
    rotationSpace=BCS_BoneSpace
    translationMode=BMM_Additive
    translationSpace=BCS_BoneSpace
    alphaBoolBlend {blendOption=Linear}
    boneToModify {boneName=C_Tail_A_Base_Jnt}
    rotation {yaw=35}
   }
   n58 = "Transform (Modify) Bone - Bone: C_Tail_A_1_Jnt"(ComponentPose=@n57, Alpha=1, Translation="0,0,0", Rotation="0,0,0", Scale="1,1,1") {
    rotationMode=BMM_Additive
    rotationSpace=BCS_BoneSpace
    translationMode=BMM_Additive
    translationSpace=BCS_BoneSpace
    alphaBoolBlend {blendOption=Linear}
    boneToModify {boneName=C_Tail_A_1_Jnt}
   }
   n59 = "Transform (Modify) Bone - Bone: C_Tail_A_2_Jnt"(ComponentPose=@n58, Alpha=1, Translation="0,0,0", Rotation="-5,0,0", Scale="1,1,1") {
    rotationMode=BMM_Additive
    rotationSpace=BCS_BoneSpace
    translationMode=BMM_Additive
    translationSpace=BCS_BoneSpace
    alphaBoolBlend {blendOption=Linear}
    boneToModify {boneName=C_Tail_A_2_Jnt}
    rotation {pitch=-5}
   }
   n60 = "Transform (Modify) Bone - Bone: C_Tail_A_3_Jnt"(ComponentPose=@n59, Alpha=1, Translation="0,0,0", Rotation="-10,0,0", Scale="1,1,1") {
    rotationMode=BMM_Additive
    rotationSpace=BCS_BoneSpace
    translationMode=BMM_Additive
    translationSpace=BCS_BoneSpace
    alphaBoolBlend {blendOption=Linear}
    boneToModify {boneName=C_Tail_A_3_Jnt}
    rotation {pitch=-10}
   }
   n61 = "Transform (Modify) Bone - Bone: C_Tail_A_4_Jnt"(ComponentPose=@n60, Alpha=1, Translation="0,0,0", Rotation="-10,0,0", Scale="1,1,1") {
    rotationMode=BMM_Additive
    rotationSpace=BCS_BoneSpace
    translationMode=BMM_Additive
    translationSpace=BCS_BoneSpace
    alphaBoolBlend {blendOption=Linear}
    boneToModify {boneName=C_Tail_A_4_Jnt}
    rotation {pitch=-10}
   }
   n62 = "Transform (Modify) Bone - Bone: C_Tail_A_5_Jnt"(ComponentPose=@n61, Alpha=1, Translation="0,0,0", Rotation="0,0,0", Scale="1,1,1") {
    rotationMode=BMM_Additive
    rotationSpace=BCS_BoneSpace
    translationMode=BMM_Additive
    translationSpace=BCS_BoneSpace
    alphaBoolBlend {blendOption=Linear}
    boneToModify {boneName=C_Tail_A_5_Jnt}
   }
   n63 = "Transform (Modify) Bone - Bone: C_Tail_A_6_Jnt"(ComponentPose=@n62, Alpha=1, Translation="0,0,0", Rotation="5,0,0", Scale="1,1,1") {
    rotationMode=BMM_Additive
    rotationSpace=BCS_BoneSpace
    translationMode=BMM_Additive
    translationSpace=BCS_BoneSpace
    alphaBoolBlend {blendOption=Linear}
    boneToModify {boneName=C_Tail_A_6_Jnt}
    rotation {pitch=5}
   }
   n64 = "Transform (Modify) Bone - Bone: C_Tail_A_7_Jnt"(ComponentPose=@n63, Alpha=1, Translation="0,0,0", Rotation="5,0,0", Scale="1,1,1") {
    rotationMode=BMM_Additive
    rotationSpace=BCS_BoneSpace
    translationMode=BMM_Additive
    translationSpace=BCS_BoneSpace
    alphaBoolBlend {blendOption=Linear}
    boneToModify {boneName=C_Tail_A_7_Jnt}
    rotation {pitch=5}
   }
   n65 = "Transform (Modify) Bone - Bone: C_Tail_A_8_Jnt"(ComponentPose=@n64, Alpha=1, Translation="0,0,0", Rotation="10,0,0", Scale="1,1,1") {
    rotationMode=BMM_Additive
    rotationSpace=BCS_BoneSpace
    translationMode=BMM_Additive
    translationSpace=BCS_BoneSpace
    alphaBoolBlend {blendOption=Linear}
    boneToModify {boneName=C_Tail_A_8_Jnt}
    rotation {pitch=10}
   }
   n66 = "Transform (Modify) Bone - Bone: C_Tail_A_9_Jnt"(ComponentPose=@n65, Alpha=1, Translation="0,0,0", Rotation="30,0,0", Scale="1,1,1") {
    rotationMode=BMM_Additive
    rotationSpace=BCS_BoneSpace
    translationMode=BMM_Additive
    translationSpace=BCS_BoneSpace
    alphaBoolBlend {blendOption=Linear}
    alphaScaleBiasClamp {bInterpResult=true}
    boneToModify {boneName=C_Tail_A_9_Jnt}
    rotation {pitch=30}
   }
   n67 = Get Is Moving 2D()
   n68 = Get Is Falling()
   n69 = Trail controller - Bone: C_Tail_A_9_Jnt!(
    ComponentPose=Local To Component(LocalPose=Use cached pose 'CachePose'())
    Alpha=Get TrailAlpha() -> (TrailAlpha:double)
    RelaxationSpeedScale:float=Get RelaxSpeed() -> (RelaxSpeed:double)
   ) -> (Pose) {
    bLimitRotation=true
    chainLength=10
    maxDeltaTime=0.0333
    alphaScaleBiasClamp {bInterpResult=true}
    rotationLimits [
     {limitMax=<180 180 180>}
     {limitMax=<180 180 180>}
     {limitMax=<180 180 180>}
     {limitMax=<180 180 180>}
     {limitMax=<180 180 180>}
     {limitMax=<180 180 180>}
     {limitMax=<180 180 180>}
     {limitMax=<180 180 180>}
     {limitMax=<180 180 180>}
     {limitMax=<180 180 180>}
    ]
    rotationOffsets [<0 0 0>, <0 0 0>, <0 0 0>, <0 0 0>, <0 0 0>, <0 0 0>, <0 0 0>, <0 0 0>, <0 0 0>, <0 0 0>]
    trailBone {boneName=C_Tail_A_9_Jnt}
    trailRelaxationSpeed {editorCurveData {keys [<at 0 10 RCIM_Linear>,<at 0.5107 8.1304 RCIM_Linear>,<at 1 5 RCIM_Linear>]}}
   }
   n70 = "Transform (Modify) Bone - Bone: C_Tail_A_Base_Jnt"(ComponentPose=@n69, Alpha=1, Translation="0,0,0", Rotation="0,30,0", Scale="1,1,1") {
    rotationMode=BMM_Additive
    rotationSpace=BCS_BoneSpace
    translationMode=BMM_Additive
    translationSpace=BCS_BoneSpace
    alphaBoolBlend {blendOption=Linear}
    boneToModify {boneName=C_Tail_A_Base_Jnt}
    rotation {yaw=30}
   }
   n71 = "Transform (Modify) Bone - Bone: C_Tail_A_1_Jnt"(ComponentPose=@n70, Alpha=1, Translation="0,0,0", Rotation="15,0,0", Scale="1,1,1") {
    rotationMode=BMM_Additive
    rotationSpace=BCS_BoneSpace
    translationMode=BMM_Additive
    translationSpace=BCS_BoneSpace
    alphaBoolBlend {blendOption=Linear}
    boneToModify {boneName=C_Tail_A_1_Jnt}
    rotation {pitch=15}
   }
   n72 = "Transform (Modify) Bone - Bone: C_Tail_A_2_Jnt"(ComponentPose=@n71, Alpha=1, Translation="0,0,0", Rotation="10,0,0", Scale="1,1,1") {
    rotationMode=BMM_Additive
    rotationSpace=BCS_BoneSpace
    translationMode=BMM_Additive
    translationSpace=BCS_BoneSpace
    alphaBoolBlend {blendOption=Linear}
    boneToModify {boneName=C_Tail_A_2_Jnt}
    rotation {pitch=10}
   }
   n73 = "Transform (Modify) Bone - Bone: C_Tail_A_3_Jnt"(ComponentPose=@n72, Alpha=1, Translation="0,0,0", Rotation="5,0,0", Scale="1,1,1") {
    rotationMode=BMM_Additive
    rotationSpace=BCS_BoneSpace
    translationMode=BMM_Additive
    translationSpace=BCS_BoneSpace
    alphaBoolBlend {blendOption=Linear}
    boneToModify {boneName=C_Tail_A_3_Jnt}
    rotation {pitch=5}
   }
   n74 = "Transform (Modify) Bone - Bone: C_Tail_A_4_Jnt"(ComponentPose=@n73, Alpha=1, Translation="0,0,0", Rotation="0,0,0", Scale="1,1,1") {
    rotationMode=BMM_Additive
    rotationSpace=BCS_BoneSpace
    translationMode=BMM_Additive
    translationSpace=BCS_BoneSpace
    alphaBoolBlend {blendOption=Linear}
    boneToModify {boneName=C_Tail_A_4_Jnt}
   }
   n75 = "Transform (Modify) Bone - Bone: C_Tail_A_5_Jnt"(ComponentPose=@n74, Alpha=1, Translation="0,0,0", Rotation="5,0,0", Scale="1,1,1") {
    rotationMode=BMM_Additive
    rotationSpace=BCS_BoneSpace
    translationMode=BMM_Additive
    translationSpace=BCS_BoneSpace
    alphaBoolBlend {blendOption=Linear}
    boneToModify {boneName=C_Tail_A_5_Jnt}
    rotation {pitch=5}
   }
   n76 = "Transform (Modify) Bone - Bone: C_Tail_A_6_Jnt"(ComponentPose=@n75, Alpha=1, Translation="0,0,0", Rotation="10,0,0", Scale="1,1,1") {
    rotationMode=BMM_Additive
    rotationSpace=BCS_BoneSpace
    translationMode=BMM_Additive
    translationSpace=BCS_BoneSpace
    alphaBoolBlend {blendOption=Linear}
    boneToModify {boneName=C_Tail_A_6_Jnt}
    rotation {pitch=10}
   }
   n77 = "Transform (Modify) Bone - Bone: C_Tail_A_7_Jnt"(ComponentPose=@n76, Alpha=1, Translation="0,0,0", Rotation="10,0,0", Scale="1,1,1") {
    rotationMode=BMM_Additive
    rotationSpace=BCS_BoneSpace
    translationMode=BMM_Additive
    translationSpace=BCS_BoneSpace
    alphaBoolBlend {blendOption=Linear}
    boneToModify {boneName=C_Tail_A_7_Jnt}
    rotation {pitch=10}
   }
   n78 = "Transform (Modify) Bone - Bone: C_Tail_A_8_Jnt"(ComponentPose=@n77, Alpha=1, Translation="0,0,0", Rotation="15,0,0", Scale="1,1,1") {
    rotationMode=BMM_Additive
    rotationSpace=BCS_BoneSpace
    translationMode=BMM_Additive
    translationSpace=BCS_BoneSpace
    alphaBoolBlend {blendOption=Linear}
    boneToModify {boneName=C_Tail_A_8_Jnt}
    rotation {pitch=15}
   }
   n79 = "Transform (Modify) Bone - Bone: C_Tail_A_9_Jnt"(ComponentPose=@n78, Alpha=1, Translation="0,0,0", Rotation="15,0,0", Scale="1,1,1") {
    rotationMode=BMM_Additive
    rotationSpace=BCS_BoneSpace
    translationMode=BMM_Additive
    translationSpace=BCS_BoneSpace
    alphaBoolBlend {blendOption=Linear}
    boneToModify {boneName=C_Tail_A_9_Jnt}
    rotation {pitch=15}
   }
   Blend Poses by bool(
    BlendPose_0=Component To Local(ComponentPose=@n79)
    BlendPose_1=Blend Poses by bool(
     BlendPose_0=Component To Local(ComponentPose=@n66)
     BlendPose_1=Blend Poses by bool(BlendPose_0=@n55, BlendPose_1=@n44, BlendTime_0=0.6, BlendTime_1=0.6, bActiveValue=@n56) {blendType=Linear}
     BlendTime_0=5
     BlendTime_1=5
     bActiveValue=AND Boolean(A=@n68, B=@n67)
    ) {blendType=Linear}
    bActiveValue=Get Is Crouching() -> (Is Crouching:bool)
   ) {blendType=Linear}
   n80 = Trail controller - Bone: C_Tail_A_9_Jnt(ComponentPose=Local To Component(LocalPose=Use cached pose 'CachePose'()), Alpha=Get TrailAlpha(), RelaxationSpeedScale=Get RelaxSpeed()) {
    bLimitRotation=true
    chainLength=10
    maxDeltaTime=0.0333
    alphaScaleBiasClamp {bInterpResult=true}
    rotationLimits [
     {limitMax=<180 180 180>}
     {limitMax=<180 180 180>}
     {limitMax=<180 180 180>}
     {limitMax=<180 180 180>}
     {limitMax=<180 180 180>}
     {limitMax=<180 180 180>}
     {limitMax=<180 180 180>}
     {limitMax=<180 180 180>}
     {limitMax=<180 180 180>}
     {limitMax=<180 180 180>}
    ]
    rotationOffsets [<0 0 0>, <0 0 0>, <0 0 0>, <0 0 0>, <0 0 0>, <0 0 0>, <0 0 0>, <0 0 0>, <0 0 0>, <0 0 0>]
    trailBone {boneName=C_Tail_A_9_Jnt}
    trailRelaxationSpeed {editorCurveData {keys [<at 0 10 RCIM_Linear>,<at 0.5107 8.1304 RCIM_Linear>,<at 1 5 RCIM_Linear>]}}
   }
   Output Pose(Result=Component To Local(ComponentPose=@n80)) {layerGroup=None}
   Save cached pose 'CachePose'(Pose=Input Pose()) {cacheName=CachePose}
  ]}
 EventGraph {nodes [
   Event BlueprintUpdateAnimation()
   > n1 = Sequence!(execute) -> (then_0, then_1, then_2:Exec, then_3:Exec)
   > Set Is Jumping(Is Jumping:bool=Get Is Jumping()) -> (Output_Get:bool)
   Set Is Crouching(execute=@n1.then_1, Is Crouching:bool=Get bIsCrouching()) -> (Output_Get:bool)
   Set Is Falling(execute=@n1.then_2, Is Falling:bool=Get bIsFalling()) -> (Output_Get:bool)
   Set Is Moving 2D(execute=@n1.then_3, Is Moving 2D:bool=Get bIsMoving2D()) -> (Output_Get:bool)
   Event OnCharacterCustomizationLoaded()
   n2 = CheckForBackbling()
   > Set isWearingBackbling(isWearingBackbling=@n2)
   > Set isWearingCape(isWearingCape=@n2.bIsBackblingLong)
   > Branch(Condition=OR Boolean(A=Get isWearingCape() -> (isWearingCape:bool), B=Get isWearingBackbling()))
   > n3 = Get Dynamic Material Instance(SkeletalMeshComponent=GetPartSkeletalMeshComponent())
   > SetScalarParameterValue(self=@n3, ParameterName=Hide Element 1, Value=1)
  ]}
]
variables [
 Is Crouching {type=boolean}
 Is Falling {type=boolean}
 Is Jumping {type=boolean}
 Is Moving 2D {type=boolean}
 isWearingBackbling {type=boolean}
 isWearingCape {type=boolean}
 RelaxSpeed {type=number}
 TrailAlpha {type=number}
]
